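# Physics asset for the hanging chains and left-side bag on the Void Redemption Rebel female medium body cosmetic.
# Capsules on the pelvis, lower spine, thighs and skirt bones form body volumes around the dynamic bones.
# Swing- and twist-limited capsule strings run from pelvis-linked base bodies, with the right front chain branching off the left's first link.
# Short counter-chains from second bases join the back and right-front chains through locked constraints, closing hanging loops.
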PhysicsAsset /BRCosmetics/Characters/Player/Female/Medium/Bodies/F_MED_VoidRedemption_Rebel/Meshes/F_MED_VoidRedemption_Rebel_Physics_Chains
bodies {
 dyn_bag_l_01 [{
   shapeType=Capsule
   length=7.5651
   radius=2.9415
   shapeName=dyn_bag_l_01_capsule
   center=<5.222 -0.0803 2.4724>
   rotation=<rot 0 -84.5692 -88.8147>
  }]
 dyn_bag_l_MSTR [{shapeType=Capsule, length=0.505, radius=0.505, shapeName=dyn_bag_l_MSTR_capsule, rotation=<rot 0 -90 0>}]
 dyn_chain_back_01_l [{
   shapeType=Capsule
   length=4.0994
   radius=0.7299
   shapeName=dyn_chain_back_01_l_capsule
   center=<2.1642 0 0>
   rotation=<rot 90 -178.7341 -178.7344>
  }]
 dyn_chain_back_02_l [{
   shapeType=Capsule
   length=4.9901
   radius=1.2
   shapeName=dyn_chain_back_02_l_capsule
   center=<4.0812 0.3857 0.0102>
   rotation=<rot 0 87.5268 -84.599>
  }]
 dyn_chain_back_03_l [{
   shapeType=Capsule
   length=3.2001
   radius=1.2
   shapeName=dyn_chain_back_03_l_capsule
   center=<4.6159 0.7387 -0.3406>
   rotation=<rot -90 -90 90.001>
  }]
 dyn_chain_back_04_1_l [{
   shapeType=Capsule
   length=3.6371
   radius=1.1633
   shapeName=dyn_chain_back_04_1_l_capsule
   center=<-3.2864 -0.0155 -0.0446>
   rotation=<rot 0 94.359 -89.5263>
  }]
 dyn_chain_back_04_2_l [{
   shapeType=Capsule
   length=2.155
   radius=1.3036
   shapeName=dyn_chain_back_04_2_l_capsule
   center=<4.6666 0.6431 -0.612>
   rotation=<rot 0.0177 -89.9896 89.6923>
  }]
 dyn_chain_back_04_l [{shapeType=Capsule, length=0.7552, radius=1.1633, shapeName=dyn_chain_back_04_l_capsule}]
 dyn_chain_back_BASE_04_l [{shapeType=Capsule, length=0.505, radius=0.505, shapeName=dyn_chain_back_BASE_04_l_capsule, rotation=<rot 0 -90 0>}]
 dyn_chain_back_BASE_l [{shapeType=Capsule, length=0.505, radius=0.505, shapeName=dyn_chain_back_BASE_l_capsule, rotation=<rot 0 -90 0>}]
 dyn_chain_front_01_l [{shapeType=Capsule, length=2, radius=1.3, shapeName=dyn_chain_front_01_l_capsule, center=<1.2678 0 0.0166>, rotation=<rot 90 180 179.9997>}]
 dyn_chain_front_01_r [{
   shapeType=Capsule
   length=3.7238
   radius=1.1436
   shapeName=dyn_chain_front_01_r_capsule
   center=<3.155 0.2091 -0.0015>
   rotation=<rot 0 -90.7544 -88.6651>
  }]
 dyn_chain_front_02_l [{
   shapeType=Capsule
   length=2
   radius=1.2449
   shapeName=dyn_chain_front_02_l_capsule
   center=<1.2678 0 0.0166>
   rotation=<rot 90 -179.4936 -179.4939>
  }]
 dyn_chain_front_02_r [{
   shapeType=Capsule
   length=3.5218
   radius=1.179
   shapeName=dyn_chain_front_02_r_capsule
   center=<3.5006 0.0421 0.2646>
   rotation=<rot 0 -91.8207 -84.662>
  }]
 dyn_chain_front_03_l [{
   shapeType=Capsule
   length=2
   radius=1.2527
   shapeName=dyn_chain_front_03_l_capsule
   center=<1.2678 0 0.0166>
   rotation=<rot 90 -173.2786 -173.2789>
  }]
 dyn_chain_front_03_r [{
   shapeType=Capsule
   length=3.02
   radius=1.1092
   shapeName=dyn_chain_front_03_r_capsule
   center=<3.4973 0.1055 0.1273>
   rotation=<rot 0 -93.1552 -86.8155>
  }]
 dyn_chain_front_04_l [{shapeType=Capsule, length=2, radius=1.2744, shapeName=dyn_chain_front_04_l_capsule, center=<1.2678 0 0.0166>, rotation=<rot 90 180 179.9997>}]
 dyn_chain_front_04_r [{
   shapeType=Capsule
   length=2.9101
   radius=1.0018
   shapeName=dyn_chain_front_04_r_capsule
   center=<4.2499 0.2762 0.1936>
   rotation=<rot 0 -93.0211 -88.4027>
  }]
 dyn_chain_front_05_l [
  {
   shapeType=Capsule
   length=6.5
   radius=1.4518
   shapeName=dyn_chain_front_05_l_capsule
   center=<4.8514 -0.062 -0.0309>
   rotation=<rot 0 -91.4292 -88.6466>
  }
  {shapeType=Capsule, length=6.5, radius=1.4518, shapeName=dyn_chain_front_05_l_capsule, center=<4.8945 0 0>}
 ]
 dyn_chain_front_05_r [{
   shapeType=Capsule
   length=3.9902
   radius=1.1603
   shapeName=dyn_chain_front_05_r_capsule
   center=<3.6282 0.2564 0.3495>
   rotation=<rot 0 -93.5456 -84.0437>
  }]
 dyn_chain_front_06_r [{shapeType=Capsule, length=2.8261, radius=1.3103, shapeName=dyn_chain_front_06_r_capsule, center=<1.8533 0.2073 0>, rotation=<rot 90 0 0>}]
 dyn_chain_front_07_1_r [{
   shapeType=Capsule
   length=2.7116
   radius=0.794
   shapeName=dyn_chain_front_07_1_r_capsule
   center=<1.7945 0.1923 -0.0608>
   rotation=<rot 0 -85.5536 -86.3611>
  }]
 dyn_chain_front_07_2_r [{
   shapeType=Capsule
   length=2.8351
   radius=0.8536
   shapeName=dyn_chain_front_07_2_r_capsule
   center=<2.6552 -0.9524 0.3537>
   rotation=<rot -1.0613 -116.467 80.7993>
  }]
 dyn_chain_front_07_r [{
   shapeType=Capsule
   length=2.3798
   radius=1.2328
   shapeName=dyn_chain_front_07_r_capsule
   center=<-1.7945 0.1923 0.0608>
   rotation=<rot 0 -94.4464 -86.3611>
  }]
 dyn_chain_front_BASE_07_r [{shapeType=Capsule, length=0.505, radius=0.505, shapeName=dyn_chain_front_BASE_07_r_capsule, rotation=<rot 0 -90 0>}]
 dyn_chain_front_BASE_l [{shapeType=Capsule, length=0.505, radius=0.505, shapeName=dyn_chain_front_BASE_l_capsule, rotation=<rot 0 -90 0>}]
 dyn_skirt_f_01_l [{
   shapeType=Capsule
   length=8.0289
   radius=9.0279
   shapeName=dyn_skirt_f_01_l_capsule
   center=<8.3647 -7.6283 0>
   rotation=<rot 90 -179.1048 178.1663>
  }]
 dyn_skirt_f_01_r [{
   shapeType=Capsule
   length=8.0289
   radius=9.0279
   shapeName=dyn_skirt_f_01_l_capsule
   center=<8.2083 -8.411 0>
   rotation=<rot -90 180 -179.9998>
  }]
 dyn_skirt_f_Collider_c [{
   shapeType=Capsule
   length=18.6149
   radius=6.7644
   shapeName=dyn_skirt_f_Collider_c_capsule
   center=<-3.4327 -4.217 0>
   rotation=<rot 90 0 9.5617>
  }]
 dyn_skirt_s_01_l [{shapeType=Capsule, length=7.4015, radius=7.8014, shapeName=dyn_skirt_s_01_l_capsule, center=<9.794 -7.1972 0>, rotation=<rot 90 0 0>}]
 pelvis [{shapeType=Capsule, length=8.882, radius=10.2707, shapeName=pelvis_capsule, center=<-1.1428 -1.1988 0>}]
 spine_01 [{shapeType=Capsule, length=6.4633, radius=9.7986, shapeName=spine_01_capsule, center=<0 -0.0166 0>, rotation=<rot 0 -90 0>}]
 thigh_l [{shapeType=Capsule, length=19.482, radius=10.0897, shapeName=thigh_l_capsule, center=<-14.6686 -1.2607 2.0052>, rotation=<rot 90 0 0>}]
 thigh_r [{shapeType=Capsule, length=21.8024, radius=9.5964, shapeName=thigh_l_capsule, center=<14.6686 1.2607 -1.5971>, rotation=<rot 90 0 180>}]
}
constraints [
 {
  bone1Name=thigh_l
  bone2Name=pelvis
  swing1LimitDegrees=45
  swing1Motion=Free
  swing2LimitDegrees=45
  swing2Motion=Free
  twistLimitDegrees=45
  twistMotion=Free
 }
 {
  bone1Name=spine_01
  bone2Name=pelvis
  swing1LimitDegrees=45
  swing1Motion=Free
  swing2LimitDegrees=45
  swing2Motion=Free
  twistLimitDegrees=45
  twistMotion=Free
 }
 {bone1Name=dyn_chain_front_01_l, bone2Name=dyn_chain_front_BASE_l, swing1LimitDegrees=85, swing2LimitDegrees=5, twistLimitDegrees=3}
 {bone1Name=dyn_chain_front_02_l, bone2Name=dyn_chain_front_01_l, swing1LimitDegrees=20, swing2LimitDegrees=10, twistLimitDegrees=10}
 {bone1Name=dyn_chain_front_03_l, bone2Name=dyn_chain_front_02_l, swing1LimitDegrees=25, swing2LimitDegrees=15, twistLimitDegrees=10}
 {bone1Name=dyn_chain_front_04_l, bone2Name=dyn_chain_front_03_l, swing1LimitDegrees=30, swing2LimitDegrees=25, twistLimitDegrees=10}
 {bone1Name=dyn_chain_front_05_l, bone2Name=dyn_chain_front_04_l, swing1LimitDegrees=30, swing2LimitDegrees=30, twistLimitDegrees=10}
 {
  bone1Name=thigh_r
  bone2Name=pelvis
  swing1LimitDegrees=45
  swing1Motion=Free
  swing2LimitDegrees=45
  swing2Motion=Free
  twistLimitDegrees=45
  twistMotion=Free
 }
 {
  bone1Name=dyn_chain_front_BASE_l
  bone2Name=pelvis
  swing1LimitDegrees=45
  swing1Motion=Free
  swing2LimitDegrees=45
  swing2Motion=Free
  twistLimitDegrees=45
  twistMotion=Free
 }
 {
  bone1Name=dyn_skirt_f_01_l
  bone2Name=thigh_l
  swing1LimitDegrees=45
  swing1Motion=Free
  swing2LimitDegrees=45
  swing2Motion=Free
  twistLimitDegrees=45
  twistMotion=Free
 }
 {bone1Name=dyn_chain_front_01_r, bone2Name=dyn_chain_front_01_l, swing1LimitDegrees=45, swing2LimitDegrees=45, twistLimitDegrees=15}
 {bone1Name=dyn_chain_front_02_r, bone2Name=dyn_chain_front_01_r, swing1LimitDegrees=45, swing2LimitDegrees=45, twistLimitDegrees=15}
 {bone1Name=dyn_chain_front_03_r, bone2Name=dyn_chain_front_02_r, swing1LimitDegrees=45, swing2LimitDegrees=45, twistLimitDegrees=15}
 {bone1Name=dyn_chain_front_04_r, bone2Name=dyn_chain_front_03_r, swing1LimitDegrees=45, swing2LimitDegrees=45, twistLimitDegrees=15}
 {bone1Name=dyn_chain_front_05_r, bone2Name=dyn_chain_front_04_r, swing1LimitDegrees=45, swing2LimitDegrees=45, twistLimitDegrees=15}
 {bone1Name=dyn_chain_front_06_r, bone2Name=dyn_chain_front_05_r, swing1LimitDegrees=45, swing2LimitDegrees=45, twistLimitDegrees=15}
 {bone1Name=dyn_chain_front_07_r, bone2Name=dyn_chain_front_06_r, swing1LimitDegrees=45, swing2LimitDegrees=45, twistLimitDegrees=45}
 {
  bone1Name=dyn_skirt_f_01_r
  bone2Name=thigh_r
  swing1LimitDegrees=45
  swing1Motion=Free
  swing2LimitDegrees=45
  swing2Motion=Free
  twistLimitDegrees=45
  twistMotion=Free
 }
 {
  bone1Name=dyn_chain_back_BASE_l
  bone2Name=pelvis
  swing1LimitDegrees=45
  swing1Motion=Free
  swing2LimitDegrees=45
  swing2Motion=Free
  twistLimitDegrees=45
  twistMotion=Free
 }
 {bone1Name=dyn_chain_back_01_l, bone2Name=dyn_chain_back_BASE_l, swing1LimitDegrees=75, swing2LimitDegrees=15, twistLimitDegrees=5}
 {bone1Name=dyn_chain_back_02_l, bone2Name=dyn_chain_back_01_l, swing1LimitDegrees=45, swing2LimitDegrees=45, twistLimitDegrees=15}
 {bone1Name=dyn_chain_back_03_l, bone2Name=dyn_chain_back_02_l, swing1LimitDegrees=45, swing2LimitDegrees=45, twistLimitDegrees=15}
 {bone1Name=dyn_chain_back_04_l, bone2Name=dyn_chain_back_03_l, swing1LimitDegrees=20, swing2LimitDegrees=20, twistLimitDegrees=10}
 {
  bone1Name=dyn_bag_l_MSTR
  bone2Name=pelvis
  swing1LimitDegrees=45
  swing1Motion=Free
  swing2LimitDegrees=45
  swing2Motion=Free
  twistLimitDegrees=45
  twistMotion=Free
 }
 {bone1Name=dyn_bag_l_01, bone2Name=dyn_bag_l_MSTR, swing1LimitDegrees=3, swing2LimitDegrees=40, twistLimitDegrees=2}
 {
  bone1Name=dyn_skirt_f_Collider_c
  bone2Name=pelvis
  swing1LimitDegrees=45
  swing1Motion=Free
  swing2LimitDegrees=45
  swing2Motion=Free
  twistLimitDegrees=45
  twistMotion=Free
 }
 {
  bone1Name=dyn_skirt_s_01_l
  bone2Name=thigh_l
  swing1LimitDegrees=45
  swing1Motion=Free
  swing2LimitDegrees=45
  swing2Motion=Free
  twistLimitDegrees=45
  twistMotion=Free
 }
 {
  bone1Name=dyn_chain_front_BASE_07_r
  bone2Name=pelvis
  swing1LimitDegrees=45
  swing1Motion=Locked
  swing2LimitDegrees=45
  swing2Motion=Locked
  twistLimitDegrees=45
  twistMotion=Locked
 }
 {
  bone1Name=dyn_chain_back_BASE_04_l
  bone2Name=pelvis
  swing1LimitDegrees=45
  swing1Motion=Free
  swing2LimitDegrees=45
  swing2Motion=Free
  twistLimitDegrees=45
  twistMotion=Free
 }
 {bone1Name=dyn_chain_back_04_1_l, bone2Name=dyn_chain_back_BASE_04_l, swing1LimitDegrees=45, swing2LimitDegrees=45, twistLimitDegrees=15}
 {bone1Name=dyn_chain_back_04_2_l, bone2Name=dyn_chain_back_04_1_l, swing1LimitDegrees=45, swing2LimitDegrees=45, twistLimitDegrees=15}
 {bone1Name=dyn_chain_front_07_1_r, bone2Name=dyn_chain_front_BASE_07_r, swing1LimitDegrees=45, swing2LimitDegrees=45, twistLimitDegrees=20}
 {bone1Name=dyn_chain_front_07_2_r, bone2Name=dyn_chain_front_07_1_r, swing1LimitDegrees=45, swing2LimitDegrees=45, twistLimitDegrees=20}
 {
  bone1Name=dyn_chain_front_07_2_r
  bone2Name=dyn_chain_front_06_r
  swing1LimitDegrees=45
  swing1Motion=Locked
  swing2LimitDegrees=45
  swing2Motion=Locked
  twistLimitDegrees=15
  twistMotion=Locked
 }
 {
  bone1Name=dyn_chain_back_04_2_l
  bone2Name=dyn_chain_back_03_l
  swing1LimitDegrees=45
  swing1Motion=Locked
  swing2LimitDegrees=45
  swing2Motion=Locked
  twistLimitDegrees=15
  twistMotion=Locked
 }
 {
  bone1Name=dyn_chain_front_07_r
  bone2Name=dyn_chain_front_07_1_r
  swing1LimitDegrees=45
  swing1Motion=Free
  swing2LimitDegrees=45
  swing2Motion=Free
  twistLimitDegrees=45
  twistMotion=Free
 }
 {
  bone1Name=dyn_chain_front_06_r
  bone2Name=dyn_chain_front_01_r
  swing1LimitDegrees=45
  swing1Motion=Free
  swing2LimitDegrees=45
  swing2Motion=Free
  twistLimitDegrees=45
  twistMotion=Free
 }
 {bone1Name=dyn_chain_front_07_r, bone2Name=dyn_chain_front_01_r, swing1LimitDegrees=45, swing2LimitDegrees=45, twistLimitDegrees=45}
]
