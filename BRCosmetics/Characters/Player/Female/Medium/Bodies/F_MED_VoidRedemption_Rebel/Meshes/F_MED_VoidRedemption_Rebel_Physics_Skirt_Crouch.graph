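# Physics asset for the crouch variant of the skirt on the female medium VoidRedemption Rebel body, giving capsule bodies to both thighs and the front, side and back skirt bones.
# Each tiny BASE body links to its thigh through a constraint with free swing and twist.
# Each larger 01 flap body hangs from its BASE with narrow angular limits, with the back flaps allowed the widest primary swing.
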PhysicsAsset /BRCosmetics/Characters/Player/Female/Medium/Bodies/F_MED_VoidRedemption_Rebel/Meshes/F_MED_VoidRedemption_Rebel_Physics_Skirt_Crouch
bodies {
 dyn_skirt_b_01_l [{
   shapeType=Capsule
   length=14.3826
   radius=2.4627
   shapeName=dyn_skirt_b_01_r_capsule
   center=<14.1298 0.2647 0.0076>
   rotation=<rot 0 0 -180>
  }]
 dyn_skirt_b_01_r [{shapeType=Capsule, length=14.3826, radius=2.4627, shapeName=dyn_skirt_b_01_r_capsule, center=<14.3036 -1.3013 0.0235>}]
 dyn_skirt_b_BASE_l [{shapeType=Capsule, length=0.505, radius=0.505, shapeName=dyn_skirt_b_BASE_l_capsule, rotation=<rot 0 -90 0>}]
 dyn_skirt_b_BASE_r [{shapeType=Capsule, length=0.505, radius=0.505, shapeName=dyn_skirt_b_BASE_r_capsule, rotation=<rot 0 -90 0>}]
 dyn_skirt_f_01_l [{shapeType=Capsule, length=8, radius=1.5, shapeName=dyn_skirt_f_01_l_capsule, center=<14 0 0>}]
 dyn_skirt_f_01_r [{shapeType=Capsule, length=8, radius=1.5, shapeName=dyn_skirt_f_01_r_capsule, center=<14 0 0>}]
 dyn_skirt_f_BASE_l [{shapeType=Capsule, length=0.505, radius=0.505, shapeName=dyn_skirt_f_BASE_l_capsule, rotation=<rot 0 -90 0>}]
 dyn_skirt_f_BASE_r [{shapeType=Capsule, length=0.505, radius=0.505, shapeName=dyn_skirt_f_BASE_r_capsule, rotation=<rot 0 -90 0>}]
 dyn_skirt_s_01_l [{shapeType=Capsule, length=8, radius=1.5, shapeName=dyn_skirt_s_01_l_capsule, center=<14 0 0>}]
 dyn_skirt_s_01_r [{shapeType=Capsule, length=8, radius=1.5, shapeName=dyn_skirt_s_01_r_capsule, center=<14 0 0>}]
 dyn_skirt_s_BASE_l [{shapeType=Capsule, length=0.505, radius=0.505, shapeName=dyn_skirt_s_BASE_l_capsule, rotation=<rot 0 -90 0>}]
 dyn_skirt_s_BASE_r [{shapeType=Capsule, length=0.505, radius=0.505, shapeName=dyn_skirt_s_BASE_r_capsule, rotation=<rot 0 -90 0>}]
 thigh_l [{
   shapeType=Capsule
   length=28.5664
   radius=7.845
   shapeName=thigh_l_capsule
   center=<-15.0501 -0.0007 0.4587>
   rotation=<rot 84.5275 179.9802 179.9915>
  }]
 thigh_r [{
   shapeType=Capsule
   length=28.5664
   radius=7.845
   shapeName=thigh_l_capsule
   center=<15.0501 0.0007 -0.4587>
   rotation=<rot 84.5275 179.9802 -0.0085>
  }]
}
constraints [
 {
  bone1Name=dyn_skirt_s_BASE_l
  bone2Name=thigh_l
  swing1LimitDegrees=45
  swing1Motion=Free
  swing2LimitDegrees=45
  swing2Motion=Free
  twistLimitDegrees=45
  twistMotion=Free
 }
 {
  bone1Name=dyn_skirt_f_BASE_l
  bone2Name=thigh_l
  swing1LimitDegrees=45
  swing1Motion=Free
  swing2LimitDegrees=45
  swing2Motion=Free
  twistLimitDegrees=45
  twistMotion=Free
 }
 {
  bone1Name=dyn_skirt_b_BASE_l
  bone2Name=thigh_l
  swing1LimitDegrees=45
  swing1Motion=Free
  swing2LimitDegrees=45
  swing2Motion=Free
  twistLimitDegrees=45
  twistMotion=Free
 }
 {
  bone1Name=dyn_skirt_b_BASE_r
  bone2Name=thigh_r
  swing1LimitDegrees=45
  swing1Motion=Free
  swing2LimitDegrees=45
  swing2Motion=Free
  twistLimitDegrees=45
  twistMotion=Free
 }
 {
  bone1Name=dyn_skirt_f_BASE_r
  bone2Name=thigh_r
  swing1LimitDegrees=45
  swing1Motion=Free
  swing2LimitDegrees=45
  swing2Motion=Free
  twistLimitDegrees=45
  twistMotion=Free
 }
 {
  bone1Name=dyn_skirt_s_BASE_r
  bone2Name=thigh_r
  swing1LimitDegrees=45
  swing1Motion=Free
  swing2LimitDegrees=45
  swing2Motion=Free
  twistLimitDegrees=45
  twistMotion=Free
 }
 {bone1Name=dyn_skirt_b_01_r, bone2Name=dyn_skirt_b_BASE_r, swing1LimitDegrees=30, swing2LimitDegrees=8, twistLimitDegrees=8}
 {bone1Name=dyn_skirt_f_01_r, bone2Name=dyn_skirt_f_BASE_r, swing1LimitDegrees=20, swing2LimitDegrees=8, twistLimitDegrees=8}
 {bone1Name=dyn_skirt_s_01_r, bone2Name=dyn_skirt_s_BASE_r, swing1LimitDegrees=20, swing2LimitDegrees=15, twistLimitDegrees=8}
 {bone1Name=dyn_skirt_f_01_l, bone2Name=dyn_skirt_f_BASE_l, swing1LimitDegrees=20, swing2LimitDegrees=8, twistLimitDegrees=8}
 {bone1Name=dyn_skirt_s_01_l, bone2Name=dyn_skirt_s_BASE_l, swing1LimitDegrees=20, swing2LimitDegrees=15, twistLimitDegrees=8}
 {bone1Name=dyn_skirt_b_01_l, bone2Name=dyn_skirt_b_BASE_l, swing1LimitDegrees=30, swing2LimitDegrees=8, twistLimitDegrees=8}
]
