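# Post-process style AnimBP for the Void Redemption Rebel female body cosmetic, adding secondary motion to zipper, collars, skirt and chains.
# The pose copied from the parent mesh passes through a deform Control Rig, bone constraints, crouch bag offset, AnimDynamics, a second Control Rig, RigidBody sims and a crouch post rig.
# The event graph switches gravity and front-end sim settings by movement state and feeds calf velocity and skydive wind into cloth WPO material parameters.
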AnimBlueprint /BRCosmetics/Characters/Player/Female/Medium/Bodies/F_MED_VoidRedemption_Rebel/Meshes/F_MED_VoidRedemption_Rebel_AnimBP
target_skeleton=&/F_MED_VoidRedemption_Rebel_Skeleton
graphs [
 AnimGraph {nodes [
   RigidBody!(ComponentPose, Alpha, OverrideWorldGravity:Vector="0,0,0", ExternalForce) -> (Pose) {
    bClampLinearTranslationLimitToRefPose=true
    bOverrideWorldGravity=true
    cachedBoundsScale=2
    overridePhysicsAsset=&/F_MED_VoidRedemption_Rebel_Physics_Skirt_Crouch
    simulationSpace=BaseBoneSpace
    alphaBoolBlend {blendOption=Linear}
    baseBoneRef {boneName=extra_skirt_belt}
    componentAppliedLinearAccClamp=<3000 3000 3000>
    componentLinearAccScale=<0.4 0.3 0.3>
    componentLinearVelScale=<0.4 0.3 0.3>
    overrideWorldGravity {z=-980}
   }
   n1 = Control Rig(Source=Copy Pose From Mesh() {bUseAttachedParent=true}) {
    alphaBoolBlend {blendOption=Linear}
    controlRigAssetReference {blueprintRigClass=&/F_MED_VoidRedemption_Rebel_CtrlRig_Deform.F_MED_VoidRedemption_Rebel_CtrlRig_Deform_C}
   }
   n2 = "Constraint - extra_skirt_belt (PP)"!(ComponentPose=Local To Component(LocalPose=@n1), Alpha=1, ConstraintWeights_0=0.65, ConstraintWeights_1:float=0.35) -> (Pose) {
    alphaBoolBlend {blendOption=Linear}
    boneToModify {boneName=extra_skirt_belt}
    constraintSetup [{targetBone {boneName=pelvis}}, {targetBone {boneName=spine_01}}]
    constraintWeights [1,1]
   }
   n3 = "Constraint - extra_body_belt (PPPP)"!(
    ComponentPose=@n2
    Alpha=1
    ConstraintWeights_0=0.018
    ConstraintWeights_1:float=0.664
    ConstraintWeights_2:float=0.3
    ConstraintWeights_3:float=0.018
   ) -> (Pose) {
    alphaBoolBlend {blendOption=Linear}
    boneToModify {boneName=extra_body_belt}
    constraintSetup [{targetBone {boneName=spine_01}}, {targetBone {boneName=spine_02}}, {targetBone {boneName=spine_03}}, {targetBone {boneName=spine_04}}]
    constraintWeights [1,1,1,1]
   }
   n4 = "Constraint - dyn_zipper_BASE_l (PPP)"!(ComponentPose=@n3, Alpha=1, ConstraintWeights_0=0.406, ConstraintWeights_1:float=0.494, ConstraintWeights_2:float=0.1) -> (Pose) {
    alphaBoolBlend {blendOption=Linear}
    boneToModify {boneName=dyn_zipper_BASE_l}
    constraintSetup [{targetBone {boneName=spine_02}}, {targetBone {boneName=spine_03}}, {targetBone {boneName=spine_04}}]
    constraintWeights [1,1,1]
   }
   n5 = "Constraint - dyn_collar_top_BASE_r (PP)"!(ComponentPose=@n4, Alpha=1, ConstraintWeights_0=0.544, ConstraintWeights_1:float=0.456) -> (Pose) {
    alphaBoolBlend {blendOption=Linear}
    boneToModify {boneName=dyn_collar_top_BASE_r}
    constraintSetup [{targetBone {boneName=spine_05}}, {targetBone {boneName=clavicle_r}}]
    constraintWeights [1,1]
   }
   n6 = "Constraint - dyn_collar_bot_BASE_r (PPP)"!(ComponentPose=@n5, Alpha=1, ConstraintWeights_0=0.145, ConstraintWeights_1:float=0.213, ConstraintWeights_2:float=0.642) -> (Pose) {
    alphaBoolBlend {blendOption=Linear}
    boneToModify {boneName=dyn_collar_bot_BASE_r}
    constraintSetup [{targetBone {boneName=spine_03}}, {targetBone {boneName=spine_04}}, {targetBone {boneName=spine_05}}]
    constraintWeights [1,1,1]
   }
   n7 = "Constraint - dyn_skirt_f_Collider_c (PP)"!(ComponentPose=@n6, Alpha=1, ConstraintWeights_0=0.5, ConstraintWeights_1:float=0.5) -> (Pose) {
    alphaBoolBlend {blendOption=Linear}
    boneToModify {boneName=dyn_skirt_f_Collider_c}
    constraintSetup [{targetBone {boneName=thigh_l}}, {targetBone {boneName=thigh_r}}]
    constraintWeights [1,1]
   }
   n8 = "Constraint - extra_collar_mid_BASE_r (PPPP)"!(
    ComponentPose=@n7
    Alpha=1
    ConstraintWeights_0=0.153
    ConstraintWeights_1:float=0.063
    ConstraintWeights_2:float=0.16
    ConstraintWeights_3:float=0.624
   ) -> (Pose) {
    alphaBoolBlend {blendOption=Linear}
    boneToModify {boneName=extra_collar_mid_BASE_r}
    constraintSetup [{targetBone {boneName=clavicle_r}}, {targetBone {boneName=spine_03}}, {targetBone {boneName=spine_04}}, {targetBone {boneName=spine_05}}]
    constraintWeights [1,1,1,1]
   }
   n9 = "Transform (Modify) Bone - Bone: dyn_bag_l_MSTR"!(ComponentPose=@n8, bAlphaBoolEnabled:bool=Get bIsCrouching(), Translation="0,-5.091,0", Rotation="-0.000,-43.168,0.000", Scale="1,1,1") -> (Pose) {
    alphaInputType=Bool
    rotationMode=BMM_Additive
    translationMode=BMM_Additive
    alphaBoolBlend {blendInTime=0.2, blendOption=Linear, blendOutTime=0.2}
    boneToModify {boneName=dyn_bag_l_MSTR}
    rotation {yaw=-43.168}
    translation {y=-5.091}
   }
   n10 = Anim Dynamics - Bone: dyn_zipper_01_l!(
    SimulationSpace:AnimPhysSimSpaceType Enum=Get AnimDynamics_SimSpace_Frontend() -> (AnimDynamics_SimSpace_Frontend:AnimPhysSimSpaceType Enum)
    ComponentPose=@n9
    Alpha=Get Anim_Dynamics_Alpha_Frontend() -> (Anim_Dynamics_Alpha_Frontend:double)
    GravityOverride:Vector=Get GravityOverride()
    ExternalForce="0,0,0"
   ) -> (Pose) {
    angularBiasOverride=0.99
    angularDampingOverride=0.99
    angularSpringConstant=40
    bAngularSpring=true
    bLinearSpring=true
    bOverrideAngularBias=true
    bOverrideAngularDamping=true
    bOverrideLinearDamping=true
    bUseGravityOverride=true
    linearDampingOverride=0.99
    linearSpringConstant=40
    simulationSpace=BoneRelative
    alphaBoolBlend {blendOption=Linear}
    boundBone {boneName=dyn_zipper_01_l}
    componentAppliedLinearAccClamp=<150 150 150>
    componentLinearAccScale=<0.1 0.75 0.75>
    componentLinearVelScale=<0.1 0.75 0.75>
    physicsBodyDefinitions [{
      boundBone {boneName=dyn_zipper_01_l}
      boxExtents=<9 2.7 5.3>
      constraintSetup {angularLimitsMax=<5 80 140>, angularLimitsMin=<-5 -80 0>, angularTarget=<0 0 0>}
      localJointOffset=<1 0 0>
     }]
    relativeSpaceBone {boneName=root}
   }
   n11 = Get AnimDynamics_SimSpace_Frontend()
   n12 = "Anim Dynamics - Chain: dyn_collar_bot_01_r -> dyn_collar_bot_02_r"!(
    SimulationSpace:AnimPhysSimSpaceType Enum=@n11
    ComponentPose=@n10
    Alpha=Get Anim_Dynamics_Alpha_Frontend()
    GravityOverride:Vector=Get GravityOverride()
    ExternalForce="0,0,0"
   ) -> (Pose) {
    angularBiasOverride=0.9
    angularDampingOverride=0.9
    angularSpringConstant=10
    bAngularSpring=true
    bChain=true
    bLinearSpring=true
    bOverrideAngularBias=true
    bOverrideAngularDamping=true
    bOverrideLinearDamping=true
    bUseGravityOverride=true
    linearDampingOverride=0.9
    linearSpringConstant=10
    simulationSpace=BoneRelative
    alphaBoolBlend {blendOption=Linear}
    boundBone {boneName=dyn_collar_bot_01_r}
    chainEnd {boneName=dyn_collar_bot_02_r}
    componentAppliedLinearAccClamp=<150 150 150>
    componentLinearAccScale=<0.1 0.75 0.75>
    componentLinearVelScale=<0.1 0.75 0.75>
    physicsBodyDefinitions [
     {
      boundBone {boneName=dyn_collar_bot_01_r}
      boxExtents=<12 4 8>
      constraintSetup {angularLimitsMax=<0 0 25>, angularLimitsMin=<0 0 -5>, angularTarget=<0 0 0>}
      localJointOffset=<1 0 0>
     }
     {
      boundBone {boneName=dyn_collar_bot_02_r}
      boxExtents=<6 2 4>
      constraintSetup {angularLimitsMax=<0 0 25>, angularLimitsMin=<0 0 -5>, angularTarget=<0 0 0>}
      localJointOffset=<0 0 0>
     }
    ]
    relativeSpaceBone {boneName=root}
   }
   n13 = "Anim Dynamics - Chain: dyn_collar_top_01_r -> dyn_collar_top_02_r"!(
    SimulationSpace:AnimPhysSimSpaceType Enum=@n11
    ComponentPose=@n12
    Alpha=Get Anim_Dynamics_Alpha_Frontend()
    GravityOverride:Vector=Get GravityOverride()
    ExternalForce="0,0,0"
   ) -> (Pose) {
    angularBiasOverride=0.95
    angularDampingOverride=0.95
    angularSpringConstant=160
    bAngularSpring=true
    bChain=true
    bLinearSpring=true
    bOverrideAngularBias=true
    bOverrideAngularDamping=true
    bOverrideLinearDamping=true
    bUseGravityOverride=true
    linearDampingOverride=0.95
    linearSpringConstant=160
    simulationSpace=BoneRelative
    alphaBoolBlend {blendOption=Linear}
    boundBone {boneName=dyn_collar_top_01_r}
    chainEnd {boneName=dyn_collar_top_02_r}
    componentAppliedLinearAccClamp=<150 150 150>
    componentLinearAccScale=<0.1 0.75 0.75>
    componentLinearVelScale=<0.1 0.75 0.75>
    physicsBodyDefinitions [
     {
      boundBone {boneName=dyn_collar_top_01_r}
      boxExtents=<16 5 10>
      constraintSetup {angularLimitsMax=<0 0 13>, angularLimitsMin=<0 0 -3>, angularTarget=<0 0 0>}
      localJointOffset=<1 0 0>
     }
     {
      boundBone {boneName=dyn_collar_top_02_r}
      boxExtents=<7.5 2.5 5>
      constraintSetup {angularLimitsMax=<0 0 13>, angularLimitsMin=<0 0 -3>, angularTarget=<0 0 0>}
      localJointOffset=<0 0 0>
     }
    ]
    relativeSpaceBone {boneName=root}
   }
   n14 = Anim Dynamics - Bone: dyn_collar_back_01_c!(
    SimulationSpace:AnimPhysSimSpaceType Enum=@n11
    ComponentPose=@n13
    Alpha=Get Anim_Dynamics_Alpha_Frontend()
    GravityOverride:Vector=Get GravityOverride()
    ExternalForce="0,0,0"
   ) -> (Pose) {
    angularBiasOverride=0.95
    angularDampingOverride=0.95
    angularSpringConstant=10
    bAngularSpring=true
    bLinearSpring=true
    bOverrideAngularBias=true
    bOverrideAngularDamping=true
    bOverrideLinearDamping=true
    bUseGravityOverride=true
    linearDampingOverride=0.95
    linearSpringConstant=10
    simulationSpace=BoneRelative
    alphaBoolBlend {blendOption=Linear}
    boundBone {boneName=dyn_collar_back_01_c}
    componentAppliedLinearAccClamp=<150 150 150>
    componentLinearAccScale=<0.1 0.75 0.75>
    componentLinearVelScale=<0.1 0.75 0.75>
    physicsBodyDefinitions [{
      boundBone {boneName=dyn_collar_back_01_c}
      boxExtents=<5 3 15>
      constraintSetup {angularLimitsMax=<0 0 5>, angularLimitsMin=<0 0 -25>, angularTarget=<0 0 0>}
      localJointOffset=<1 0 0>
     }]
    relativeSpaceBone {boneName=root}
   }
   n15 = "Anim Dynamics - Chain: dyn_collar_01_l -> dyn_collar_02_l"!(
    SimulationSpace:AnimPhysSimSpaceType Enum=@n11
    ComponentPose=@n14
    Alpha=Get Anim_Dynamics_Alpha_Frontend()
    GravityOverride:Vector=Get GravityOverride()
    ExternalForce="0,0,0"
   ) -> (Pose) {
    angularBiasOverride=0.95
    angularDampingOverride=0.95
    angularSpringConstant=160
    bAngularSpring=true
    bChain=true
    bLinearSpring=true
    bOverrideAngularBias=true
    bOverrideAngularDamping=true
    bOverrideLinearDamping=true
    bUseGravityOverride=true
    linearDampingOverride=0.95
    linearSpringConstant=160
    simulationSpace=BoneRelative
    alphaBoolBlend {blendOption=Linear}
    boundBone {boneName=dyn_collar_01_l}
    chainEnd {boneName=dyn_collar_02_l}
    componentAppliedLinearAccClamp=<150 150 150>
    componentLinearAccScale=<0.1 0.75 0.75>
    componentLinearVelScale=<0.1 0.75 0.75>
    physicsBodyDefinitions [
     {
      boundBone {boneName=dyn_collar_01_l}
      boxExtents=<16 5 10>
      constraintSetup {angularLimitsMax=<0 0 13>, angularLimitsMin=<0 0 -3>, angularTarget=<0 0 0>}
      localJointOffset=<2 0 0>
     }
     {
      boundBone {boneName=dyn_collar_02_l}
      boxExtents=<7.5 2.5 5>
      constraintSetup {angularLimitsMax=<0 0 13>, angularLimitsMin=<0 0 -3>, angularTarget=<0 0 0>}
      localJointOffset=<1 0 0>
     }
    ]
    relativeSpaceBone {boneName=root}
   }
   n16 = Control Rig!(Source=Component To Local(ComponentPose=@n15), Alpha, isNotMoving2D:bool=Get isNotMoving2D() -> (isNotMoving2D:bool)) -> (Pose) {
    alphaBoolBlend {blendOption=Linear}
    controlRigAssetReference {blueprintRigClass=&/F_MED_VoidRedemption_Rebel_CtrlRig.F_MED_VoidRedemption_Rebel_CtrlRig_C}
   }
   n17 = RigidBody(ComponentPose=Local To Component(LocalPose=@n16), OverrideWorldGravity=Get GravityOverride()) {
    bClampLinearTranslationLimitToRefPose=true
    bOverrideWorldGravity=true
    cachedBoundsScale=2
    overridePhysicsAsset=&/F_MED_VoidRedemption_Rebel_Physics
    simulationSpace=BaseBoneSpace
    alphaBoolBlend {blendOption=Linear}
    baseBoneRef {boneName=extra_skirt_belt}
    componentAppliedLinearAccClamp=<3000 3000 3000>
    componentLinearAccScale=<0.4 0.3 0.3>
    componentLinearVelScale=<0.4 0.3 0.3>
    overrideWorldGravity {z=-980}
   }
   n18 = Get GravityOverrideChain() -> (GravityOverrideChain:Vector)
   n19 = RigidBody(ComponentPose=@n17, OverrideWorldGravity=@n18) {
    bClampLinearTranslationLimitToRefPose=true
    bOverrideWorldGravity=true
    cachedBoundsScale=1
    overridePhysicsAsset=&/F_MED_VoidRedemption_Rebel_Physics_Chains
    simulationSpace=BaseBoneSpace
    alphaBoolBlend {blendOption=Linear}
    baseBoneRef {boneName=extra_skirt_belt}
    componentLinearAccScale=<0.3 0.3 0.3>
    componentLinearVelScale=<0.3 0.3 0.3>
    overrideWorldGravity {z=-980}
    simSpaceSettings {worldAlpha=0.1}
   }
   n20 = RigidBody(ComponentPose=@n19, OverrideWorldGravity=@n18) {
    bClampLinearTranslationLimitToRefPose=true
    bOverrideWorldGravity=true
    cachedBoundsScale=1
    overridePhysicsAsset=&/F_MED_VoidRedemption_Rebel_Physics_Chains_crouch
    simulationSpace=BaseBoneSpace
    alphaBoolBlend {blendOption=Linear}
    baseBoneRef {boneName=extra_skirt_belt}
    componentLinearAccScale=<0.3 0.3 0.3>
    componentLinearVelScale=<0.3 0.3 0.3>
    overrideWorldGravity {z=-980}
    simSpaceSettings {worldAlpha=0.1}
   }
   n21 = Control Rig!(Source=Component To Local(ComponentPose=@n20), Alpha, is_NOT_Moving_2d:double=Get is_NOT_Moving_2d() -> (is_NOT_Moving_2d:double)) -> (Pose) {
    alphaBoolBlend {blendOption=Linear}
    controlRigAssetReference {blueprintRigClass=&/F_MED_VoidRedemption_Rebel_CtrlRig_Crouch_post.F_MED_VoidRedemption_Rebel_CtrlRig_Crouch_post_C}
   }
   Output Pose(Result=@n21) {layerGroup=None}
  ]}
 EventGraph {nodes [
   n1 = Event BlueprintUpdateAnimation()
   > n2 = Sequence!(execute) -> (then_0, then_1, then_2:Exec, then_3:Exec, then_4:Exec)
   n3 = Branch(execute=@n2.then_1, Condition=Get bIsSkyDiving())
   > Set Velocity Interp Speed_reg1(Velocity Interp Speed_reg1:double=500) -> (Output_Get:double)
   > Set Velocity Interp Speed_reg2(Velocity Interp Speed_reg2:double=500) -> (Output_Get:double)
   > Set wind_strength(wind_strength:float=10) -> (Output_Get:float)
   > Set TurbulenceNoiseTileDivisor(TurbulenceNoiseTileDivisor:float=50) -> (Output_Get:float)
   Set Velocity Interp Speed_reg1(execute=@n3.else, Velocity Interp Speed_reg1=0)
   > Set Velocity Interp Speed_reg2(Velocity Interp Speed_reg2=0)
   > Set wind_strength(wind_strength=0)
   > Set TurbulenceNoiseTileDivisor(TurbulenceNoiseTileDivisor=0)
   Set WPO Globals Function(
    execute=@n2.then_2
    Material=Get dynamic_mat_instance() -> (dynamic_mat_instance:Material Instance Dynamic Object Reference)
    Overlay Material=Get dynamic_overlay_mat() -> (dynamic_overlay_mat:Material Instance Dynamic Object Reference)
    Skeletal Mesh=Get mesh() -> (mesh:Skeletal Mesh Component Budgeted Object Reference)
    "Local Noise Transform  Location"="0, 0, 0"
    Local Noise Offset="0, 0, 0"
   )
   > Set WPO Regional Function!(
    execute
    Material:Material Instance Dynamic Object Reference=Get dynamic_mat_instance()
    Overlay Material:Material Instance Dynamic Object Reference=Get dynamic_overlay_mat()
    Skeletal Mesh:Skeletal Mesh Component Object Reference=Get mesh()
    Velocity Transform:Name=calf_r
    Velocity Interp Speed:double=Get Velocity Interp Speed_reg1() -> (Velocity Interp Speed_reg1:double)
    Collision Transform:Name=calf_r
    Collision Transform Axis:EAxis Enum
    Invert Collision Transform Axis:bool=false
    Delta Time X:double=@n1.DeltaTimeX
    Previous Velocity Transform Location:"Vector (by ref)"=Get Previous Velocity Transform Location_reg1() -> (Previous Velocity Transform Location_reg1:Vector)
    Previous Velocity Transform Velocity:"Vector (by ref)"=Get Previous Velocity Transform Velocity_reg1() -> (Previous Velocity Transform Velocity_reg1:Vector)
    Region:int=1
   ) -> (then)
   > Set WPO Regional Function(
    Material=Get dynamic_mat_instance()
    Overlay Material=Get dynamic_overlay_mat()
    Skeletal Mesh=Get mesh()
    Velocity Transform=calf_l
    Velocity Interp Speed=Get Velocity Interp Speed_reg2() -> (Velocity Interp Speed_reg2:double)
    Collision Transform=calf_l
    Invert Collision Transform Axis=true
    Delta Time X=@n1.DeltaTimeX
    Previous Velocity Transform Location=Get Previous Velocity Transform Location_reg2() -> (Previous Velocity Transform Location_reg2:Vector)
    Previous Velocity Transform Velocity=Get Previous Velocity Transform Velocity_reg2() -> (Previous Velocity Transform Velocity_reg2:Vector)
    Region=2
   )
   n4 = Get dynamic_mat_instance()
   n5 = Get dynamic_overlay_mat()
   > SetScalarParameterValue(self=[@n4, @n5], ParameterName=ClothSim_WindStrength, Value=Get wind_strength() -> (wind_strength:float))
   > SetScalarParameterValue(
    self=[@n4, @n5]
    ParameterName=ClothSim_TurbulenceNoiseTileDivisor
    Value=Get TurbulenceNoiseTileDivisor() -> (TurbulenceNoiseTileDivisor:float)
   )
   n6 = Gravity Override State Function DEPRECATED(
    execute=@n2.then_0
    CharacterPartAnimInstance=Self-Reference()
    GameTimeInSeconds=GetGameTimeInSeconds()
    GravityOverrideParams=Get Gravity Override Params()
   )
   > n7 = Set GravityOverride(GravityOverride=@n6)
   n8 = Get bIsSkydiveDiveMode()
   > Set GravityOverrideChain(GravityOverrideChain:Vector=SelectVector(A="0,-200,-600", B=SelectVector(A="-750,0,0", B=@n7, bPickA=@n8), bPickA=Get bIsParachuteOpened())) -> (Output_Get:Vector)
   n9 = NOT Boolean(A=Get bIsMoving2D())
   Set is_NOT_Moving_2d(execute=@n2.then_3, is_NOT_Moving_2d:double="To Float (Boolean)"(InBool=@n9)) -> (Output_Get:double)
   > Set isNotMoving2D(isNotMoving2D:bool=@n9) -> (Output_Get:bool)
   > Set isNotCrouching(isNotCrouching:bool=NOT Boolean(A=Get bIsCrouching())) -> (Output_Get:bool)
   n10 = NOT Boolean(A=Get bIsPlayingEmote())
   n11 = NOT Boolean(A=Get bIsSkyDiving())
   n12 = Branch(execute=@n2.then_4, Condition=AND Boolean(A=Get bIsInFrontEnd(), B=AND Boolean(A=@n11, B=@n10)))
   > Set AnimDynamics_SimSpace_Frontend(AnimDynamics_SimSpace_Frontend:AnimPhysSimSpaceType Enum=World) -> (Output_Get:AnimPhysSimSpaceType Enum)
   > Set Anim_Dynamics_Alpha_Frontend(Anim_Dynamics_Alpha_Frontend:double=0.5) -> (Output_Get:double)
   Set AnimDynamics_SimSpace_Frontend(execute=@n12.else, AnimDynamics_SimSpace_Frontend=BoneRelative)
   > Set Anim_Dynamics_Alpha_Frontend(Anim_Dynamics_Alpha_Frontend=1)
   n13 = Event OnCharacterCustomizationLoaded()
   > n14 = Set mesh(mesh:Skeletal Mesh Component Budgeted Object Reference=Get Array Item with EFort Custom Part Type(Array=@n13.CustomMeshParts)) -> (Output_Get:Skeletal Mesh Component Budgeted Object Reference)
   > n15 = Get Dynamic Material Instance(SkeletalMeshComponent=@n14)
   > Set dynamic_mat_instance(dynamic_mat_instance:Material Instance Dynamic Object Reference=@n15) -> (Output_Get:Material Instance Dynamic Object Reference)
   > n16 = GetOrCreateDynamicOverlayMaterial(MeshComponent=Get mesh(), MaterialName=None)
   > Set dynamic_overlay_mat(dynamic_overlay_mat:Material Instance Dynamic Object Reference=@n16) -> (Output_Get:Material Instance Dynamic Object Reference)
  ]}
]
variables [
 Anim_Dynamics_Alpha_Frontend {type=number}
 AnimDynamics_SimSpace_Frontend {type=AnimPhysSimSpaceType}
 dynamic_mat_instance {type=MaterialInstanceDynamic}
 dynamic_overlay_mat {type=MaterialInstanceDynamic}
 Gravity Override Params {type=GravityOverrideParamsStruct}
 GravityOverride {type=Vector}
 GravityOverrideChain {type=Vector}
 is_NOT_Moving_2d {type=number}
 isNotCrouching {type=boolean}
 isNotMoving2D {type=boolean}
 mesh {type=SkeletalMeshComponentBudgeted}
 Previous Velocity Transform Location_reg1 {type=Vector}
 Previous Velocity Transform Location_reg2 {type=Vector}
 Previous Velocity Transform Velocity_reg1 {type=Vector}
 Previous Velocity Transform Velocity_reg2 {type=Vector}
 TurbulenceNoiseTileDivisor {type=number}
 Velocity Interp Speed_reg1 {type=number, default=1}
 Velocity Interp Speed_reg2 {type=number, default=1}
 wind_strength {type=number}
]
