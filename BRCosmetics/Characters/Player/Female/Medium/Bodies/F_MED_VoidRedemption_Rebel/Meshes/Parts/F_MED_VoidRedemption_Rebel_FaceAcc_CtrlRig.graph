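# Control Rig for the F_MED_VoidRedemption_Rebel face accessory that procedurally offsets ponytail and hair bones as the head moves, only when RigidBody Enabled is false.
# Spherical pose readers track head against spine_05: one, Kalman-smoothed, weights a dyn_tail_back_02_c offset, the other shifts the three back hair bones.
# Head-to-spine_05 Euler Z, remapped to a weight, drives offsets on both side hair base bones.
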ControlRigBlueprint /BRCosmetics/Characters/Player/Female/Medium/Bodies/F_MED_VoidRedemption_Rebel/Meshes/Parts/F_MED_VoidRedemption_Rebel_FaceAcc_CtrlRig
graphs [
 RigVMFunctionLibrary {links [], nodes []}
 RigVMModel {nodes [
   RigUnit_BeginExecution = &RigUnit_BeginExecution()
   RigVMFunction_ControlFlowBranch = &RigVMFunction_ControlFlowBranch(ExecuteContext=@RigUnit_BeginExecution.ExecutePin, Condition=RigVMVariableNode() {variable=RigidBody Enabled}.Value)
   n1 = &RigUnit_SphericalPoseReader(
    ActiveRegionScaleFactors=(PositiveWidth=3,NegativeWidth=0.7,PositiveHeight=3.5,NegativeHeight=2)
    ActiveRegionSize=0.1
    Debug=(bDrawDebug=false,bDraw2D=False,bDrawLocalAxes=False,DebugScale=25,DebugSegments=20,DebugThickness=0.25)
    DriverAxis=(X=0,Y=1,Z=0)
    DriverItem=(Type=Bone,Name="head")
    FalloffRegionScaleFactors=(PositiveWidth=1,NegativeWidth=1,PositiveHeight=3,NegativeHeight=0.5)
    FalloffSize=0.4
    FlipHeightScaling=False
    FlipWidthScaling=False
    OptionalParentItem=(Type=Bone,Name="spine_05")
    RotationOffset=(X=90,Y=0,Z=-60)
    ExecutePin=@RigVMFunction_ControlFlowBranch.False
   )
   KalmanFilter = &RigVMFunction_KalmanFloat(BufferSize=16, Value=@n1.OutputParam) {template="KalmanFilter::Execute(in Value,in BufferSize,out Result)"}
   > /Script/ControlRig.RigUnit_SetRelativeTransformForItem(
    Child=(Type=Bone,Name="dyn_tail_back_02_c")
    Parent=(Type=Bone,Name="dyn_tail_back_02_c")
    Value=(Rotation=(X=-0.1601,Y=-0.5269,Z=0.0757,W=0.8313),Translation=(X=-2,Y=0,Z=7),Scale3D=(X=1,Y=1,Z=1))
    bParentInitial=False
    bPropagateToChildren=True
    Weight=@KalmanFilter.Result
   ) {
    is=struct RigVMUnitNode
    template="Set Relative Transform::Execute(in Child,in Parent,in bParentInitial,in Value,in Weight,in bPropagateToChildren)"
   }
   > n2 = &RigUnit_SphericalPoseReader(
    ActiveRegionScaleFactors=(PositiveWidth=3,NegativeWidth=0.7,PositiveHeight=3.5,NegativeHeight=2)
    ActiveRegionSize=0.1
    Debug=(bDrawDebug=false,bDraw2D=False,bDrawLocalAxes=False,DebugScale=25,DebugSegments=20,DebugThickness=0.25)
    DriverAxis=(X=0,Y=1,Z=0)
    DriverItem=(Type=Bone,Name="head")
    FalloffRegionScaleFactors=(PositiveWidth=1,NegativeWidth=0.5,PositiveHeight=3,NegativeHeight=0.5)
    FalloffSize=0.4
    FlipHeightScaling=False
    FlipWidthScaling=False
    OptionalParentItem=(Type=Bone,Name="spine_05")
    RotationOffset=(X=90,Y=0,Z=-45)
   )
   > &RigUnit_SetRelativeTransformForItem(
    Child=(Type=Bone,Name="dyn_hair_back_01_c")
    Parent=(Type=Bone,Name="dyn_hair_back_01_c")
    Value=(Rotation=(X=0,Y=0,Z=0,W=1),Translation=(X=-1.3,Y=0,Z=0),Scale3D=(X=1,Y=1,Z=1))
    bParentInitial=False
    bPropagateToChildren=True
    Weight=@n2.OutputParam
   ) {template="Set Relative Transform::Execute(in Child,in Parent,in bParentInitial,in Value,in Weight,in bPropagateToChildren)"}
   > &RigUnit_SetRelativeTransformForItem(
    Child=(Type=Bone,Name="dyn_hair_back_02_c")
    Parent=(Type=Bone,Name="dyn_hair_back_02_c")
    Value=(Rotation=(X=0,Y=0,Z=0,W=1),Translation=(X=-1.3,Y=0,Z=0),Scale3D=(X=1,Y=1,Z=1))
    bParentInitial=false
    bPropagateToChildren=True
    Weight=@n2.OutputParam
   ) {template="Set Relative Transform::Execute(in Child,in Parent,in bParentInitial,in Value,in Weight,in bPropagateToChildren)"}
   > &RigUnit_SetRelativeTransformForItem(
    Child=(Type=Bone,Name="dyn_hair_back_03_c")
    Parent=(Type=Bone,Name="dyn_hair_back_03_c")
    Value=(Rotation=(X=0,Y=0,Z=0,W=1),Translation=(X=-1.3,Y=0,Z=0),Scale3D=(X=1,Y=1,Z=1))
    bParentInitial=false
    bPropagateToChildren=True
    Weight=@n2.OutputParam
   ) {template="Set Relative Transform::Execute(in Child,in Parent,in bParentInitial,in Value,in Weight,in bPropagateToChildren)"}
   RigUnit_GetRelativeTransformForItem = &RigUnit_GetRelativeTransformForItem(Child=(Type=Bone,Name="head"), Parent=(Type=Bone,Name="spine_05"), bChildInitial=False, bParentInitial=False)
   RigVMFunction_MathQuaternionToEuler = &RigVMFunction_MathQuaternionToEuler(RotationOrder=ZYX, Value=@RigUnit_GetRelativeTransformForItem.RelativeTransform.Rotation)
   Remap = &RigVMFunction_MathDoubleRemap(SourceMaximum=25, SourceMinimum=10, TargetMaximum=1, TargetMinimum=0, Value=@RigVMFunction_MathQuaternionToEuler.Result.Z)
   > &RigUnit_SetRelativeTransformForItem(
    Child=(Type=Bone,Name="dyn_hair_side_BASE_l")
    Parent=(Type=Bone,Name="dyn_hair_side_BASE_l")
    Value=(Rotation=(X=0.0001,Y=-0.1763,Z=-0.0002,W=0.9843),Translation=(X=-0.5538,Y=-1.0583,Z=1.4916),Scale3D=(X=1,Y=1,Z=1))
    bParentInitial=false
    bPropagateToChildren=True
    Weight=@Remap.Result
   ) {template="Set Relative Transform::Execute(in Child,in Parent,in bParentInitial,in Value,in Weight,in bPropagateToChildren)"}
   > &RigUnit_SetRelativeTransformForItem(
    Child=(Type=Bone,Name="dyn_hair_side_BASE_r")
    Parent=(Type=Bone,Name="dyn_hair_side_BASE_r")
    Value=(Rotation=(X=0,Y=-0.0147,Z=-0.0001,W=0.9999),Translation=(X=-0.5309,Y=-0.3725,Z=1.5136),Scale3D=(X=1,Y=1,Z=1))
    bParentInitial=false
    bPropagateToChildren=True
    Weight=@Remap.Result
   ) {template="Set Relative Transform::Execute(in Child,in Parent,in bParentInitial,in Value,in Weight,in bPropagateToChildren)"}
  ]}
]
hierarchy [
 root {loc=<0 0 0>, rot=<0 0 0 1>}
 pelvis {parent=root, loc=<0 -0.5301 98.0571>, rot=<0.0127 0.707 -0.0127 -0.707>}
 spine_01 {parent=pelvis, loc=<6.1208 -0.9328 0>, rot=<0 0 0.034 -0.9994>}
 spine_02 {parent=spine_01, loc=<6.5996 -0.2095 0>, rot=<0 0 0.0238 -0.9997>}
 spine_03 {parent=spine_02, loc=<6.4914 -0.1237 0>, rot=<0 0 0.0149 -0.9999>}
 spine_04 {parent=spine_03, loc=<6.9598 0.0794 0>, rot=<0 0 -0.0132 -0.9999>}
 spine_05 {parent=spine_04, loc=<6.9987 0 0>, rot=<0 0 -0.0015 -1>}
 neck_01 {parent=spine_05, loc=<11.3138 0.1529 0>, rot=<0 0 -0.2003 -0.9797>}
 neck_02 {parent=neck_01, loc=<4.5627 0 0>, rot=<0 0 0.0504 -0.9987>}
 head {parent=neck_02, loc=<5.3914 0 0>, rot=<0 0 0.115 -0.9934>}
 dyn_hair_side_BASE_l {parent=head, loc=<-4.2907 0.2075 -6.2289>, rot=<0.0677 -0.7721 -0.6286 -0.0649>}
 dyn_hair_side_01_l {parent=dyn_hair_side_BASE_l, loc=<0 0 0>, rot=<0 0 -0.1604 -0.9871>}
 dyn_hair_side_02_l {parent=dyn_hair_side_01_l, loc=<2.9311 0 0>, rot=<0 0 0 1>}
 dyn_hair_side_03_l {parent=dyn_hair_side_02_l, loc=<2.6283 0 0>, rot=<0 0 0 1>}
 dyn_hair_side_BASE_r {parent=head, loc=<-4.7858 1.5556 6.5383>, rot=<0.0828 -0.8458 -0.5269 -0.01>}
 dyn_hair_side_01_r {parent=dyn_hair_side_BASE_r, loc=<0 0 0>, rot=<0 0 0 1>}
 dyn_hair_side_02_r {parent=dyn_hair_side_01_r, loc=<2.9311 0 0>, rot=<0 0 0 1>}
 dyn_hair_side_03_r {parent=dyn_hair_side_02_r, loc=<2.6283 0 0>, rot=<0 0 0 1>}
 dyn_tail_back_BASE_c {parent=head, loc=<13.2905 -6.1892 -0.8314>, rot=<-0.0098 -0.0145 0.6412 -0.7671>}
 dyn_tail_back_01_c {parent=dyn_tail_back_BASE_c, loc=<0 0 0>, rot=<1 0 0 0>}
 dyn_tail_back_02_c {parent=dyn_tail_back_01_c, loc=<4.0466 0 0>, rot=<-0.0038 0.0063 -0.6142 -0.7891>}
 dyn_tail_back_03_c {parent=dyn_tail_back_02_c, loc=<4.5331 0 0>, rot=<0.0507 -0.2361 -0.2553 -0.9362>}
 dyn_tail_back_04_c {parent=dyn_tail_back_03_c, loc=<6.3164 0 0>, rot=<-0.0177 0.0532 0.0464 -0.9973>}
 dyn_tail_back_05_c1 {parent=dyn_tail_back_04_c, loc=<7.7923 -0.5232 0.0408>, rot=<-0.2523 0.4942 0.182 -0.8118>}
 dyn_hair_back_BASE_c {parent=head, loc=<-3.1711 -5.9542 0>, rot=<0.1095 -0.994 0 0>}
 dyn_hair_back_01_c {parent=dyn_hair_back_BASE_c, loc=<0 0 0>, rot=<0 0 0 1>}
 dyn_hair_back_02_c {parent=dyn_hair_back_01_c, loc=<2.9613 0 0>, rot=<0 0 0 1>}
 dyn_hair_back_03_c {parent=dyn_hair_back_02_c, loc=<2.3024 0 0>, rot=<0 0 0 1>}
 clavicle_l {parent=spine_05, loc=<5.4831 -0.1053 -3.2946>, rot=<0.0465 -0.7289 -0.009 -0.6829>}
 clavicle_r {parent=spine_05, loc=<5.4831 -0.1053 3.2946>, rot=<0.7289 0.0465 -0.6829 0.009>}
]
variables [RigidBody Enabled {type=bool, default=False}]
