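# Physics asset for the Rebel female medium body's FaceAcc part, covering a back hair tail and left and right side hair strands.
# Capsules form a five-link tail and two three-link side chains, each rooted in a small BASE body with free rotation against the head.
# Links below the bases have angular limits with mostly tight twist.
# Head, neck, clavicle and upper-spine capsules, plus a spine box, provide body shapes around the chains.
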PhysicsAsset /BRCosmetics/Characters/Player/Female/Medium/Bodies/F_MED_VoidRedemption_Rebel/Meshes/Parts/F_MED_VoidRedemption_Rebel_FaceAcc_Physics
bodies {
 clavicle_l [{
   shapeType=Capsule
   length=12.9176
   radius=5.4327
   shapeName=clavicle_l_capsule
   center=<5.1762 -1.8792 -0.8441>
   rotation=<rot 83.5501 -0.0004 -0.0004>
  }]
 clavicle_r [{
   shapeType=Capsule
   length=8.9176
   radius=4.9327
   shapeName=clavicle_l_capsule
   center=<-5.3607 1.8792 0.0291>
   rotation=<rot 84.5997 -0.0006 179.9994>
  }]
 dyn_hair_side_01_l [{
   shapeType=Capsule
   length=3.7533
   radius=0.7623
   shapeName=dyn_hair_side_01_l_capsule
   center=<1.8791 0.0076 -0.3751>
   rotation=<rot 5.2388 -90 -0.3767>
  }]
 dyn_hair_side_01_r [{
   shapeType=Capsule
   length=2.1073
   radius=0.6289
   shapeName=dyn_hair_side_01_l_capsule
   center=<1.886 0.2166 -0.6073>
   rotation=<rot 0.5868 -89.9694 179.6248>
  }]
 dyn_hair_side_02_l [{
   shapeType=Capsule
   length=3.8458
   radius=0.6075
   shapeName=dyn_hair_side_02_l_capsule
   center=<1.068 -0.1361 -0.3053>
   rotation=<rot 3.7139 -90 1.0811>
  }]
 dyn_hair_side_02_r [{
   shapeType=Capsule
   length=2.3786
   radius=0.6075
   shapeName=dyn_hair_side_02_l_capsule
   center=<1.213 0.0811 -0.5414>
   rotation=<rot -3.1185 -90.1288 -178.9196>
  }]
 dyn_hair_side_03_l [{
   shapeType=Capsule
   length=4.6007
   radius=1.1159
   shapeName=dyn_hair_side_03_l_capsule
   center=<1.3742 -0.003 0.5612>
   rotation=<rot 13.8758 -90 30.4479>
  }]
 dyn_hair_side_03_r [{
   shapeType=Capsule
   length=4.0356
   radius=0.8582
   shapeName=dyn_hair_side_03_l_capsule
   center=<0.7268 -0.1669 0.9387>
   rotation=<rot -4.7405 -94.4294 -155.6558>
  }]
 dyn_hair_side_BASE_l [{shapeType=Capsule, length=0.505, radius=0.505, shapeName=dyn_hair_side_BASE_l_capsule, rotation=<rot 0 -90 0>}]
 dyn_hair_side_BASE_r [{shapeType=Capsule, length=0.505, radius=0.505, shapeName=dyn_hair_side_BASE_l_capsule, rotation=<rot 0 -90 180>}]
 dyn_tail_back_01_c [{shapeType=Capsule, length=0, radius=2.9232, shapeName=dyn_tail_back_01_c_capsule}]
 dyn_tail_back_02_c [{shapeType=Capsule, length=1.5877, radius=3.3939, shapeName=dyn_tail_back_02_c_capsule}]
 dyn_tail_back_03_c [{
   shapeType=Capsule
   length=3.5845
   radius=2.2288
   shapeName=dyn_tail_back_03_c_capsule
   center=<2.3873 0 0>
   rotation=<rot 90 179.9999 179.9998>
  }]
 dyn_tail_back_04_c [{
   shapeType=Capsule
   length=3.6846
   radius=2.5065
   shapeName=dyn_tail_back_04_c_capsule
   center=<2.1779 0 0>
   rotation=<rot -90 -6.7215 6.7224>
  }]
 dyn_tail_back_05_c1 [{
   shapeType=Capsule
   length=4.686
   radius=1.3813
   shapeName=dyn_tail_back_05_c1_capsule
   center=<2.318 0.1969 0.3754>
   rotation=<rot 0 90.3704 -80.5857>
  }]
 dyn_tail_back_BASE_c [{shapeType=Capsule, length=0.505, radius=0.505, shapeName=dyn_tail_back_BASE_c_capsule, rotation=<rot 0 -90 0>}]
 head [
  {shapeType=Capsule, length=11.6268, radius=5.8368, shapeName=head_capsule, center=<1.5296 0.3098 0>, rotation=<rot -90 180 149.9997>}
  {shapeType=Capsule, length=10.0799, radius=3.9623, shapeName=head_capsule, center=<-0.0215 -3.9903 0>}
 ]
 neck_01 [{shapeType=Capsule, length=0.4375, radius=4.6028, shapeName=neck_01_capsule, center=<-0.5901 0.7955 0>, rotation=<rot -90 180 180>}]
 neck_02 [{shapeType=Capsule, length=0, radius=4, shapeName=neck_02_capsule, center=<0 0.6226 0>, rotation=<rot -90 0 0>}]
 spine_05 [
  {shapeType=Capsule, length=8.4374, radius=9.9864, shapeName=spine_05_capsule, center=<-3.9939 0.1764 0>}
  {
   shapeType=Capsule
   length=26.1303
   radius=4.8273
   shapeName=spine_05_capsule
   center=<7.4417 -6.1119 0.2407>
   rotation=<rot -14.4799 -0.0002 0>
  }
  {shapeType=Capsule, length=9, radius=9, shapeName=spine_05_capsule, center=<0.7614 -1.2444 0>}
  {
   shapeType=Box
   extentX=18.107
   extentY=3.3521
   extentZ=8.8353
   shapeName=spine_05_box
   center=<5.5014 4.7205 5.4576>
   rotation=<rot 22.934 -42.1519 11.4746>
  }
 ]
}
constraints [
 {
  bone1Name=head
  bone2Name=neck_02
  swing1LimitDegrees=45
  swing1Motion=Free
  swing2LimitDegrees=45
  swing2Motion=Free
  twistLimitDegrees=45
  twistMotion=Free
 }
 {
  bone1Name=dyn_tail_back_BASE_c
  bone2Name=head
  swing1LimitDegrees=45
  swing1Motion=Free
  swing2LimitDegrees=45
  swing2Motion=Free
  twistLimitDegrees=45
  twistMotion=Free
 }
 {bone1Name=dyn_tail_back_01_c, bone2Name=dyn_tail_back_BASE_c, swing1LimitDegrees=10, swing2LimitDegrees=10, twistLimitDegrees=5}
 {bone1Name=dyn_tail_back_02_c, bone2Name=dyn_tail_back_01_c, swing1LimitDegrees=45, swing2LimitDegrees=45, twistLimitDegrees=5}
 {bone1Name=dyn_tail_back_03_c, bone2Name=dyn_tail_back_02_c, swing1LimitDegrees=45, swing2LimitDegrees=45, twistLimitDegrees=5}
 {bone1Name=dyn_tail_back_04_c, bone2Name=dyn_tail_back_03_c, swing1LimitDegrees=45, swing2LimitDegrees=45, twistLimitDegrees=5}
 {bone1Name=dyn_tail_back_05_c1, bone2Name=dyn_tail_back_04_c, swing1LimitDegrees=45, swing2LimitDegrees=45, twistLimitDegrees=5}
 {
  bone1Name=dyn_hair_side_BASE_l
  bone2Name=head
  swing1LimitDegrees=45
  swing1Motion=Free
  swing2LimitDegrees=45
  swing2Motion=Free
  twistLimitDegrees=45
  twistMotion=Free
 }
 {
  bone1Name=dyn_hair_side_BASE_r
  bone2Name=head
  swing1LimitDegrees=45
  swing1Motion=Free
  swing2LimitDegrees=45
  swing2Motion=Free
  twistLimitDegrees=45
  twistMotion=Free
 }
 {bone1Name=dyn_hair_side_01_l, bone2Name=dyn_hair_side_BASE_l, swing1LimitDegrees=35, swing2LimitDegrees=45, twistLimitDegrees=7}
 {bone1Name=dyn_hair_side_02_l, bone2Name=dyn_hair_side_01_l, swing1LimitDegrees=35, swing2LimitDegrees=20, twistLimitDegrees=7}
 {bone1Name=dyn_hair_side_03_l, bone2Name=dyn_hair_side_02_l, swing1LimitDegrees=35, swing2LimitDegrees=20, twistLimitDegrees=7}
 {bone1Name=dyn_hair_side_01_r, bone2Name=dyn_hair_side_BASE_r, swing1LimitDegrees=35, swing2LimitDegrees=40, twistLimitDegrees=7}
 {bone1Name=dyn_hair_side_02_r, bone2Name=dyn_hair_side_01_r, swing1LimitDegrees=35, swing2LimitDegrees=20, twistLimitDegrees=7}
 {bone1Name=dyn_hair_side_03_r, bone2Name=dyn_hair_side_02_r, swing1LimitDegrees=35, swing2LimitDegrees=20, twistLimitDegrees=20}
]
